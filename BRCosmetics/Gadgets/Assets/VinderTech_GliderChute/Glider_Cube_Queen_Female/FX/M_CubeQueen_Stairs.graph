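# Translucent unlit material for the stair effect on the Cube Queen glider, built for Niagara mesh particles.
# Emissive blends camera-vector triplanar reflection with noise-distorted stair texture and curve gradient, adds UV1 edge glow, then applies particle color and time-of-day tint.
# Opacity subtracts actor-relative sphere masks with noise-animated borders, remapped through curves.
# A scalar switches to dissolving stairs, using static masks and vertex-color-driven world position offset outside a sphere.
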Material /BRCosmetics/Gadgets/Assets/VinderTech_GliderChute/Glider_Cube_Queen_Female/FX/M_CubeQueen_Stairs
graph {
 properties {
  bUsedWithMeshDeformer=true
  bUsedWithNiagaraMeshParticles=true
  bUsedWithSkeletalMesh=true
  blendMode=BLEND_Translucent
  opacityMaskClipValue=0
  pixelDepthOffsetMode=PDOM_Legacy
  shadingModel=MSM_Unlit
  usedShadingModels=Unlit
 }
 nodes [
  n1 = TextureCoordinate() {uTiling=1.5, vTiling=1.5}
  Add(A=@n1) {constB=0}
  n2 = LinearInterpolate(Alpha=VertexColor().R)
  n3 = ActorPositionWS()
  n4 = Transform(None=Constant3Vector() {constant=<rgba 0 -200 0 0>}) {transformSourceType=TRANSFORMSOURCE_Local}
  n5 = OneMinus(None=SphereMask(A=WorldPosition().XYZ, B=Add(A=@n3, B=@n4), Radius=Constant() {r=550}, Hardness=Constant() {r=0.7}))
  n6 = Multiply(
   A=Transform(None=VectorParameter() {group=WPO, parameterName=WPO_Z_Direction, defaultValue=<rgba 0 0 1 1>}.RGB) {transformSourceType=TRANSFORMSOURCE_Local}
   B=Multiply(A=Multiply(A=@n2, B=@n5), B=ScalarParameter() {defaultValue=20, group=WPO, parameterName=Z_WPO_Strength})
  )
  n7 = LinearInterpolate(Alpha=VertexColor().G) {constA=-1}
  n8 = Multiply(
   A=Transform(None=VectorParameter() {group=WPO, parameterName=WPO_Y_Direction, defaultValue=<rgba 0 1 0 1>}.RGB) {transformSourceType=TRANSFORMSOURCE_Local}
   B=Multiply(A=ScalarParameter() {defaultValue=5, group=WPO, parameterName=Y_WPO_Strength}, B=Multiply(A=@n5, B=@n7))
  )
  n9 = Transform(None=VectorParameter() {group=WPO, parameterName=WPO_X_Direction, defaultValue=<rgba 1 0 0 1>}.RGB) {transformSourceType=TRANSFORMSOURCE_Local}
  n10 = VertexColor()
  n11 = Multiply(
   A=ScalarParameter() {defaultValue=5, group=WPO, parameterName=X_WPO_Strength}
   B=Multiply(A=@n5, B=LinearInterpolate(Alpha=@n10.B) {constA=-1})
  )
  LinearInterpolate(
   A=Constant()
   B=Add(A=Add(A=@n6, B=@n8), B=Multiply(A=@n9, B=@n11))
   Alpha=ScalarParameter() {group=Alpha, parameterName=Switch_to_the_dissolving_stairs}
  )
  n12 = ParticleColor()
  n13 = WorldPosition()
  n14 = ParticlePositionWS()
  n15 = Constant() {r=400}
  n16 = Power(Base=OneMinus(None=SphereMask(A=@n13.XYZ, B=@n14, Radius=@n15) {attenuationRadius=0.5, hardnessPercent=0}))
  n17 = VectorParameter() {group=Cube Queen Emissive, parameterName=Edge_Color, defaultValue=<rgba 0 13.0644 20 1>}
  n18 = Multiply(A=@n17.RGB, B=Constant() {r=0.1}) {constB=0.1}
  n19 = Multiply(A=@n17.RGB, B=Constant() {r=0.2})
  n20 = TextureCoordinate() {coordinateIndex=1}
  n21 = OneMinus(None=Sine(None=ComponentMask(None=@n20) {g=true}) {period=2})
  n22 = Constant() {r=10}
  n23 = OneMinus(None=Sine(None=ComponentMask(None=@n20) {r=true}) {period=2})
  n24 = Add(A=@n22) {constB=10}
  n25 = Add(A=Saturate(None=Power(Base=@n21, Exp=@n22) {constExponent=30}), B=Saturate(None=Power(Base=@n23, Exp=@n24) {constExponent=40})) {constB=0}
  n26 = TextureObjectParameter() {
   group=Cube Queen Emissive
   parameterName=Reflection_Tex
   texture=/Game/Gadgets/Assets/VinderTech_GliderChute/Glider_Cube_Queen_Female/FX/T_Cheap_blue_space_color
  }
  n27 = Constant3Vector() {constant=<rgba 1 1 1 0>}
  n28 = TextureCoordinate()
  n29 = DynamicParameter() {defaultValue=<rgba 0 0 0 1>, paramNames [Tex_Offset,Noise_U,Noise_V,Param4]}
  n30 = AppendVector(A=@n29.Noise_U, B=@n29.Noise_U)
  n31 = VectorParameter() {group=Cube Queen Emissive, parameterName=Speed, defaultValue=<rgba 0.05 0.05 0 0>}
  n32 = TextureSample(UVs=Panner(Coordinate=Multiply(A=@n28, B=@n30), Speed=ComponentMask(None=@n31.RGB) {g=true, r=true}) {speedX=0.05, speedY=0.05}) {samplerType=SAMPLERTYPE_Grayscale, texture=&LowResBlurredNoise}
  n33 = Multiply(A=@n32.R) {constB=0.5}
  n34 = DynamicParameter() {parameterIndex=1, defaultValue=<rgba 0 0 0 1>, paramNames [Tex_U_Scale,Tex_V_Scale,Param3,Param4]}
  n35 = AppendVector(A=@n34.Tex_U_Scale, B=@n34.Tex_U_Scale)
  n36 = TextureCoordinate()
  n37 = Add(
   A=Add(A=Multiply(A=@n35, B=@n36), B=DynamicParameter() {defaultValue=<rgba 0 0 0 1>, paramNames [Tex_Offset,Noise_U,Noise_V,Param4]}.Tex_Offset) {constB=0}
   B=@n32.R
  )
  n38 = TextureSampleParameter2D(UVs=@n37) {
   group=Cube Queen Emissive
   parameterName=Colorful_Tex_Base
   texture=/Game/Gadgets/Assets/VinderTech_GliderChute/Glider_Cube_Queen_Female/FX/T_CubeQueen_Glider_Stair
  }
  n39 = CurveAtlasRowParameter(CurveTime=Power(
    Base=TextureSample(UVs=@n37) {samplerType=SAMPLERTYPE_Grayscale, texture=&LowResBlurredNoise}.B
    Exp=Constant() {r=1.5}
   ) {constExponent=1.5}) {
   atlas=/Game/Athena/LinearColorCurves/VFX_CurveAtlas
   curve=/Game/Gadgets/Assets/VinderTech_GliderChute/Glider_Cube_Queen_Female/FX/Cube_Queen_StairCuve
   defaultValue=42
   group=Cube Queen Emissive
   parameterName=NoiseGradient
  }
  n40 = LinearInterpolate(
   A=Multiply(A=@n38.RGB, B=ScalarParameter() {defaultValue=10, group=Cube Queen Emissive, parameterName=Colorful_Tex_intensity}) {constB=10}
   B=Multiply(A=@n39) {constB=1.5}
   Alpha=ScalarParameter() {defaultValue=0.3, group=Cube Queen Emissive, parameterName=Lerp_Colorful_Tex}
  ) {constAlpha=0.3}
  n41 = Constant() {r=-1}
  n42 = LinearInterpolate(
   A=Multiply(
    A=MaterialFunctionCall!(CameraVector, Texture=@n26, Tiling=@n27, Offset=@n33, AxisFadeContrast) {materialFunction=/Engine/Functions/Engine_MaterialFunctions01/Texturing/TriplanarCameraVector}.Result
    B=ScalarParameter() {defaultValue=5, group=Cube Queen Emissive, parameterName=Reflection_Tex_Intensity}
   ) {constB=5}
   B=Multiply(
    A=VectorParameter() {group=Cube Queen Emissive, parameterName=Color, defaultValue=<rgba 1 1 1 1>}.RGB
    B=Desaturation(None=@n40, Fraction=@n41)
   )
   Alpha=ScalarParameter() {defaultValue=0.8, group=Cube Queen Emissive, parameterName=Lerp_Base_color}
  )
  n43 = Add(A=Saturate(None=Multiply(A=@n16, B=@n18)), B=Add(A=Multiply(A=@n19, B=@n25), B=Add(B=@n42)))
  n44 = VectorParameter() {group=Cube Queen Emissive, parameterName=Color_Morning, defaultValue=<rgba 0.3281 0.2136 0.222 1>}
  n45 = VectorParameter() {group=Cube Queen Emissive, parameterName=Color_Day, defaultValue=<rgba 1 1 1 1>}
  n46 = VectorParameter() {group=Cube Queen Emissive, parameterName=Color_Evening, defaultValue=<rgba 0.4115 0.4115 0.4115 1>}
  n47 = VectorParameter() {group=Cube Queen Emissive, parameterName=Color_Night, defaultValue=<rgba 0.8385 0.4629 0.7708 1>}
  MaterialFunctionCall(Emissive=Multiply(
    A=Multiply(A=@n43, B=@n12.RGB)
    B=MaterialFunctionCall(Color_Morning=@n44.RGB, Color_Day=@n45.RGB, Color_Evening=@n46.RGB, Color_Night=@n47.RGB) {materialFunction=&Color_Time_of_Day}.Result
   )) {materialFunction=&MF_FX_EmissiveAdjustments}
  n48 = ActorPositionWS()
  n49 = Constant() {r=-20}
  n50 = Constant() {r=20}
  n51 = Add(A=LinearInterpolate(Alpha=ParticleRandom()) {constA=5.7, constB=5.3}, B=Time())
  n52 = Panner(Coordinate=TextureCoordinate(), Time=@n51) {speedY=0.5}
  n53 = Multiply(A=LinearInterpolate(
    A=TextureSample(UVs=@n52) {samplerType=SAMPLERTYPE_Grayscale, texture=&LowResBlurredNoise}.R
    B=TextureSample() {texture=/Game/Weapons/Prototype/ObjectMover/Textures/T_Movetool_Transition}.B
   ))
  n54 = Transform(None=Add(
    A=VectorParameter() {group=Alpha, parameterName=Stair_mask_offset, defaultValue=<rgba 0 -100 0 1>}.RGB
    B=LinearInterpolate(A=@n49, B=@n50, Alpha=@n53) {constA=-50, constB=0}
   ) {constB=0}) {transformSourceType=TRANSFORMSOURCE_Local}
  n55 = Constant() {r=500}
  n56 = Subtract(A=@n55, B=Constant() {r=50}) {constB=50}
  n57 = Panner(Coordinate=TextureCoordinate(), Time=@n51) {speedX=0.05, speedY=0.2}
  n58 = Panner(Coordinate=TextureCoordinate(), Time=@n51) {speedX=0.03, speedY=0.3}
  n59 = LinearInterpolate(
   A=TextureSample(UVs=Panner(Coordinate=@n1, Time=@n51) {speedX=-0.05, speedY=0.5}) {texture=&T_Movetool_Transition}.B
   B=LinearInterpolate(
    A=TextureSample(UVs=@n57) {samplerType=SAMPLERTYPE_LinearColor, texture=/Game/Global/Textures/Default/Noise/T_Atlas_DigitalNoise_01}.R
    B=TextureSample(UVs=@n58) {texture=&T_Movetool_Transition}.B
    Alpha=MaterialFunctionCall() {materialFunction=&DebugTimeSine}.Result
   )
  )
  n60 = CurveAtlasRowParameter(CurveTime=SphereMask(A=WorldPosition().XYZ, B=Add(A=@n48, B=@n54), Radius=LinearInterpolate(A=@n56, B=@n55, Alpha=@n59) {constA=250}, Hardness=Constant() {r=0.7})) {
   atlas=&VFX_CurveAtlas
   curve=/Game/Gadgets/Assets/VinderTech_GliderChute/Glider_Cube_Queen_Female/FX/Cube_Queen_StairMask_Cuve
   defaultValue=43
   group=Alpha
  }
  n61 = ActorPositionWS()
  n62 = Constant() {r=-20}
  n63 = Constant() {r=20}
  n64 = Add(A=LinearInterpolate(Alpha=ParticleRandom()) {constA=5.7, constB=5.3}, B=Time())
  n65 = Panner(Coordinate=TextureCoordinate(), Time=@n64) {speedY=0.5}
  n66 = Multiply(A=LinearInterpolate(
    A=TextureSample(UVs=@n65) {samplerType=SAMPLERTYPE_Grayscale, texture=&LowResBlurredNoise}.R
    B=TextureSample() {texture=&T_Movetool_Transition}.B
   ))
  n67 = Transform(None=Add(
    A=VectorParameter() {group=Alpha, parameterName=Stair_mask_offset_Top, defaultValue=<rgba 0 -400 0 1>}.RGB
    B=LinearInterpolate(A=@n62, B=@n63, Alpha=@n66) {constA=-50, constB=0}
   ) {constB=0}) {transformSourceType=TRANSFORMSOURCE_Local}
  n68 = Constant() {r=500}
  n69 = Subtract(A=@n68, B=Constant() {r=50}) {constB=50}
  n70 = TextureCoordinate() {uTiling=1.5, vTiling=1.5}
  n71 = Panner(Coordinate=TextureCoordinate(), Time=@n64) {speedX=0.05, speedY=0.2}
  n72 = Panner(Coordinate=TextureCoordinate(), Time=@n64) {speedX=0.03, speedY=0.3}
  n73 = LinearInterpolate(
   A=TextureSample(UVs=Panner(Coordinate=@n70, Time=@n64) {speedX=-0.05, speedY=0.5}) {texture=&T_Movetool_Transition}.B
   B=LinearInterpolate(
    A=TextureSample(UVs=@n71) {samplerType=SAMPLERTYPE_LinearColor, texture=&T_Atlas_DigitalNoise_01}.R
    B=TextureSample(UVs=@n72) {texture=&T_Movetool_Transition}.B
    Alpha=MaterialFunctionCall() {materialFunction=&DebugTimeSine}.Result
   )
  )
  n74 = CurveAtlasRowParameter(CurveTime=SphereMask(A=WorldPosition().XYZ, B=Add(A=@n61, B=@n67), Radius=LinearInterpolate(A=@n69, B=@n68, Alpha=@n73) {constA=250}, Hardness=Constant() {r=0.7})) {
   atlas=&VFX_CurveAtlas
   curve=&Cube_Queen_StairMask_Cuve
   defaultValue=43
   group=Alpha
  }
  n75 = Saturate(None=@n74.R)
  n76 = Constant() {r=3}
  n77 = ActorPositionWS()
  n78 = Transform(None=VectorParameter() {group=Alpha, parameterName=Dissipate stairs mask offset, defaultValue=<rgba 0 -100 0 1>}.RGB) {transformSourceType=TRANSFORMSOURCE_Local}
  n79 = CurveAtlasRowParameter(CurveTime=SphereMask(A=WorldPosition().XYZ, B=Add(A=@n77, B=@n78), Radius=Constant() {r=500}, Hardness=Constant() {r=0.7})) {
   atlas=&VFX_CurveAtlas
   curve=&Cube_Queen_StairMask_Cuve
   defaultValue=43
   group=Alpha
  }
  n80 = WorldPosition()
  n81 = Add(
   A=ActorPositionWS()
   B=Transform(None=VectorParameter() {group=Alpha, parameterName=Dissipate stairs mask offset_top, defaultValue=<rgba 0 -200 0 1>}.RGB) {transformSourceType=TRANSFORMSOURCE_Local}
  )
  n82 = Constant() {r=550}
  n83 = Constant() {r=0.7}
  n84 = LinearInterpolate(
   A=Subtract(A=Saturate(None=@n60.R), B=Power(Base=@n75, Exp=@n76))
   B=Subtract(A=Saturate(None=@n79.R), B=SphereMask(A=@n80.XYZ, B=@n81, Radius=@n82, Hardness=@n83))
   Alpha=ScalarParameter() {group=Alpha, parameterName=Switch_to_the_dissolving_stairs}
  )
  n85 = Constant() {r=1}
  n86 = ScalarParameter() {group=Alpha, parameterName=No_Use_Mask}
  Multiply(
   A=Multiply(A=LinearInterpolate(A=@n84, B=@n85, Alpha=@n86), B=@n12.A)
   B=MaterialFunctionCall(Value_Morning=@n44.A, Value_Day=@n45.A, Value_Evening=@n46.A, Value_Night=@n47.A) {materialFunction=&Scalar_Time_of_Day}.Result
  )
  ObjectPositionWS()
 ]
}
