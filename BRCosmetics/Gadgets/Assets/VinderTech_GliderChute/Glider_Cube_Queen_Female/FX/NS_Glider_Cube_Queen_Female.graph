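# Niagara effect for the Cube Queen female glider, layering stairs, deploy trail, flames, smoke and floating cubes.
# Many emitters index skeletal mesh bones by spawn group, zeroing particle size when per-frame movement exceeds a threshold.
# Region masks limit flames, smoke and cubes spatially.
# Front-end/gameplay modifiers switch or tune spawn rates and brightness, time-of-day scalars adjust fading, and User.IsSwing hides range cubes.
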NiagaraSystem /BRCosmetics/Gadgets/Assets/VinderTech_GliderChute/Glider_Cube_Queen_Female/FX/NS_Glider_Cube_Queen_Female
dependencies {
 usedDataInterfaces [
  &NiagaraDataInterfaceCurve
  &NiagaraDataInterfaceArrayFloat
  &NiagaraDataInterfaceColorCurve
  &NiagaraDataInterfaceDebugDraw
  &NiagaraDataInterfaceArrayFloat3
 ]
 usedDynamicInputs []
 usedModules [&SystemState]
}
emitters {
 Character_cube {
  emitterUpdateScript [
   &EmitterState(
    Delay First Loop Only=true
    Enable Distance Culling=true
    Inactive Response=::NewEnumerator1
    Life Cycle Mode=::NewEnumerator1
    Loop Delay=1.5
    MaxDistance=3000
    Module.bUseMaxDistance=true
    Module.bUseMinDistance=false
    Scalability Mode=::NewEnumerator1
    UseLoopDelay=true
   )
   &SpawnRate(SpawnRate=&FN_FrontEndFloatModifier(Gameplay=6, Front End=0), Use Spawn Probability=false) {module=SpawnRate001}
  ]
  particleSpawnScript [
   &InitializeParticle(
    Color=<rgba 1.5 1.5 1.5 0.8>
    Lifetime Min=1.5
    Lifetime Mode=::NewEnumerator1
    Mesh Scale Mode=::NewEnumerator1
    Mesh Uniform Scale Max=6
    Mesh Uniform Scale Min=4
    Sprite Size Mode=::NewEnumerator3
    Uniform Sprite Size=20
    UsePositionOffset=false
    Write Material Random=true
   )
   &SkeletalMeshLocation(
    Bone Sampling Mode=::NewEnumerator4
    MaskSpawnGroup=false
    Orientation Sampling=::NewEnumerator1
    Sampled Orientation Coordinate Space=::Local
    Skeletal Mesh=DataInterface() {propertyValues {
      DefaultMesh=null
      ExcludeBoneName=root
      OptionalComponentClass=null
      PreviewMesh=/BRCosmetics/Gadgets/Assets/VinderTech_GliderChute/Glider_Cube_Queen_Female/Meshes/Cube_Queen_Female_Glider
      SkinningMode=SkinOnTheFly
      SoftSourceActor=null
      SourceMode=Default
      UvSetIndex=0
      WholeMeshLOD=-1
      bExcludeBone=false
      bReadDeformedGeometry=true
      bRequireCurrentFrameData=true
      ComponentTags []
      FilteredBones [Root]
      FilteredSockets []
      MeshUserParameter {parameter {name=None, typeDefHandle {registeredTypeIndex=1}, varData []}}
      SamplingRegions []
     }}
   ) {module=SkeletalMeshLocation001}
   &TorusLocation(
    Align Torus With Axis=false
    Handle Radius=0
    Large Radius=42
    MaskSpawnGroup=false
    Module.Override Randomness=false
    Offset=&RandomRangeVector(
     Minimum=&RandomRangeVector(Minimum=<0 0 0>, Maximum=<0 0 40>)
     Maximum=&MakeVector(
      X=0
      Y=0
      Z=&FloatFromCurve(
       FloatCurve=DataInterface() {propertyValues {Curve {keys [
           {interpMode=RCIM_Cubic, tangentMode=RCTM_User, time=0, value=20}
           {arriveTangent=107.1069, interpMode=RCIM_Cubic, leaveTangent=107.1069, tangentMode=RCTM_User, time=0.2, value=60}
           <at 1 80>
          ]}}}
       CurveIndex=$Emitter.NormalizedLoopAge
      )
     )
    )
   ) {module=TorusLocation001}
   &AddVelocity(Velocity=&RandomRangeVector(Minimum=<0 0 20>, Maximum=<0 0 30>))
   &InitialMeshOrientation(
    Rotation=&RandomRangeVector(Minimum=<0 0 0>, Maximum=<1 1 1>)
    Use Orientation Vector=true
    Use Rotation Vector=true
   )
  ]
  particleUpdateScript [
   &ParticleState()
   /Niagara/Modules/Update/Velocity/VortexVelocity(
    Module.Use Influence Falloff=false
    Velocity Amount:NiagaraFloat=&FloatFromCurve(
     FloatCurve=DataInterface() {propertyValues {Curve {keys [
         <at 0 -350>
         {arriveTangent=40.078, interpMode=RCIM_Cubic, leaveTangent=40.078, tangentMode=RCTM_User, time=1.8, value=-75}
         {interpMode=RCIM_Cubic, tangentMode=RCTM_User, time=5, value=-50}
        ]}}}
     CurveIndex=$Emitter.Age
    )
   ) {module=VortexVelocity001}
   &Drag(Drag=5, Use Linear Drag=true, Use Rotational Drag=false)
   &SolveForcesAndVelocity(Clamp Velocity=false, Limit Acceleration=false)
   &ScaleColor(
    Scale Alpha=&Multiply_Float(
     A=1
     B=&FloatFromCurve(FloatCurve=DataInterface() {propertyValues {Curve {keys [
          {interpMode=RCIM_Cubic, tangentMode=RCTM_User, time=0, value=0}
          {interpMode=RCIM_Cubic, tangentMode=RCTM_User, time=0.5, value=1}
          {tangentMode=RCTM_User, time=1, value=0}
         ]}}})
    )
    ScaleA=true
    ScaleRGB=true
   )
   &UpdateMeshOrientation(
    Rotation Rate=&FloatFromCurve(
     FloatCurve=DataInterface() {propertyValues {Curve {keys [
         {arriveTangent=-2.1159, interpMode=RCIM_Cubic, leaveTangent=-2.1159, tangentMode=RCTM_User, time=0, value=1}
         {interpMode=RCIM_Cubic, tangentMode=RCTM_Break, time=1, value=0.2}
        ]}}}
     Scale Curve=&RandomRangeFloat(Minimum=0.7, Maximum=1.5)
    )
    Rotation Vector=&RandomRangeVector(Minimum=<-1 -1 -1>, Maximum=<1 1 1>)
   )
   &ScaleMeshSize(Scale Factor=&Multiply_VectorByFloat(
     Vector=<1 1 1>
     Float=&FloatFromCurve(FloatCurve=DataInterface() {propertyValues {Curve {keys [
          {interpMode=RCIM_Cubic, tangentMode=RCTM_User, time=0, value=0}
          {interpMode=RCIM_Cubic, tangentMode=RCTM_User, time=0.5, value=1}
          {tangentMode=RCTM_User, time=1, value=0}
         ]}}})
    ))
  ]
  rendererClasses [&NiagaraMeshRendererProperties]
  renderers [{rendererClass=&NiagaraMeshRendererProperties, rendererIndex=0}]
 }
 Deploy_Stairs {
  emitterUpdateScript [
   /SetVariables(Emitter.Stairs_Mask_Offset:Vector4f=&MakeVector4(X=0, Y=&FN_FrontEndFloatModifier(Gameplay=-100, Front End=-100), Z=0, W=1)) {isSetParameters=true}
   /SetVariables(Emitter.Stairs_Mask_Offset_Top:Vector4f=&MakeVector4(X=0, Y=&FN_FrontEndFloatModifier(Gameplay=-340, Front End=-340), Z=0, W=1)) {isSetParameters=true}
   &EmitterState(Inactive Response=::NewEnumerator1, Life Cycle Mode=::NewEnumerator1, Loop Behavior=::NewEnumerator1, Loop Duration=3, UseLoopDelay=false)
   &SpawnBurst_Instantaneous(Spawn Count=10, Use Spawn Probability=false, UseLoopCountLimit=false)
  ]
  particleSpawnScript [
   &FN_SetSpawnGroup(Number of Spawn Groups=10)
   &InitializeParticle(Lifetime=1.5, Mesh Scale Mode=::NewEnumerator4, UsePositionOffset=false, Write Material Random=true)
   /SetVariables() {isSetParameters=true, module=SetVariables_#1}
  ]
  particleUpdateScript [
   &ParticleState()
   &SkeletalMeshLocation(Bone / Socket Index=$Particles.SpawnGroup, Bone Sampling Mode=::NewEnumerator4, MaskSpawnGroup=false, Skeletal Mesh=$User.SK_Mesh)
   &ScaleMeshSize(Scale Factor=&Multiply_VectorByFloat(
     Vector=<1 1 1>
     Float=&MakeCustomFloatFromBool(
      Bool=&SetBoolByFloatComparison(
       A=&VectorLength(Vector=&Subtract_Vector(A=Linked() {linkedVariable {name=Particles.Position, type=NiagaraPosition}}, B=$Particles.Lastposition))
       B=100
      )
      True Float=0
      False Float=1
     )
    )) {module=ScaleMeshSize002}
   /SetVariables(Particles.Lastposition:NiagaraPosition=$Particles.SkeletalMeshLocation.SampledPosition) {isSetParameters=true, module=SetVariables_#2}
   &ScaleColor(
    Scale Alpha=&FloatFromCurve(FloatCurve=DataInterface() {propertyValues {Curve {keys [<at 1.1 0>,<at 1.3 1>,<at 5 1>]}}}, CurveIndex=$Particles.Age)
    Scale RGB=&Multiply_VectorByFloat(Vector=<1 1 1>, Float=&FN_FrontEndFloatModifier(Gameplay=1, Front End=1.5))
    ScaleA=true
    ScaleRGB=true
   )
   &DynamicMaterialParameters(
    Index 0 Param 1=&SelectFloatFromArray(
     Direct Array Index=$Particles.SpawnGroup
     Float Selection Array=DataInterface() {propertyValues {FloatData [0.3,0.7,0,0,0,0,0.6,0.3,0.5,0.7]}}
    )
    Index 0 Param 2=&SelectFloatFromArray(
     Direct Array Index=$Particles.SpawnGroup
     Float Selection Array=DataInterface() {propertyValues {FloatData [0.7,0.7,0.8,0.8,1.1,0.89,1,1,1,1]}}
    )
    Index 0 Param 3=&SelectFloatFromArray(
     Direct Array Index=$Particles.SpawnGroup
     Float Selection Array=DataInterface() {propertyValues {FloatData [0.8,1.1,0.85,0.8,0.9,1.1,1,1,1,1]}}
    )
    Index 1 Param 1=&SelectFloatFromArray(Direct Array Index=$Particles.SpawnGroup, Float Selection Array=DataInterface() {propertyValues {FloatData [0.5,0.7,1.2,1,1,1.2,1,1,1,1]}})
    Index 1 Param 2=&SelectFloatFromArray(Direct Array Index=$Particles.SpawnGroup, Float Selection Array=DataInterface() {propertyValues {FloatData [0.5,1.2,1.5,1,1,0.7,1,1,1,1]}})
    Module.Param0WriteEnabled=true
    Module.Param0WriteEnabled1=true
    Module.Param1WriteEnabled=true
    Module.Param1WriteEnabled1=true
    Module.Param2WriteEnabled=true
    Module.Param2WriteEnabled1=true
    Module.Param3WriteEnabled=true
    Module.Param3WriteEnabled1=true
    Write Parameter Index 1=true
   ) {module=DynamicMaterialParameters001}
  ]
  rendererClasses [&NiagaraMeshRendererProperties]
  renderers [{rendererClass=&NiagaraMeshRendererProperties, rendererIndex=0}]
 }
 Deploy_Trail {
  emitterUpdateScript [
   &EmitterState(
    Life Cycle Mode=::NewEnumerator1
    Loop Behavior=::NewEnumerator1
    Loop Delay=&FN_FrontEndFloatModifier(Gameplay=1, Front End=0.8)
    Loop Duration=&FN_FrontEndFloatModifier(Gameplay=0.45, Front End=0.7)
    UseLoopDelay=true
   )
   &SpawnRate(SpawnRate=&FN_FrontEndFloatModifier(Gameplay=30, Front End=30), Use Spawn Probability=false)
  ]
  particleSpawnScript [
   &FN_SetSpawnGroup(Number of Spawn Groups=7)
   &InitializeParticle(
    Color=<rgba 0 15 13.661 5>
    Lifetime=0.38
    Mesh Scale Mode=::NewEnumerator3
    Mesh Uniform Scale=1.3
    UsePositionOffset=false
    Write Material Random=true
   )
   &InitialMeshOrientation(
    Mesh Orientation Mode=::NewEnumerator2
    Rotation=&SelectVectorFromArray(
     Direct Array Index=$Particles.SpawnGroup
     Vector Selection Array=DataInterface() {propertyValues {FloatData [<0 0 0>, <0 0 0.5>, <0 0 0>, <0 0 0.5>, <0 0 0>, <0 0 0.5>, <0 0 0>]}}
    )
    Use Rotation Vector=true
   )
  ]
  particleUpdateScript [
   &ParticleState()
   &SolveForcesAndVelocity(Clamp Velocity=false, Limit Acceleration=false)
   &ScaleColor(
    Scale Alpha=&Multiply_Float(
     A=&FloatFromCurve(FloatCurve=DataInterface() {propertyValues {Curve {keys [{interpMode=RCIM_Cubic, tangentMode=RCTM_User, time=0, value=1}, {interpMode=RCIM_Cubic, tangentMode=RCTM_User, time=1, value=0}]}}})
     B=1
    )
    ScaleA=true
    ScaleRGB=true
   )
   /SetVariables(Particles.Mesh02_Color:LinearColor=&MakeLinearColorFromVectorAndFloat(
     "Vector (RGB)"=&MakeVectorFromLinearColorRGB(Color=<rgba 0.4862 0.0624 2 5>)
     "Float (Alpha)"=&Multiply_Float(
      A=&FloatFromCurve(FloatCurve=DataInterface() {propertyValues {Curve {keys [{interpMode=RCIM_Cubic, tangentMode=RCTM_User, time=0, value=1}, {interpMode=RCIM_Cubic, tangentMode=RCTM_User, time=1, value=0}]}}})
      B=3
     )
    )) {isSetParameters=true}
   &SkeletalMeshLocation(
    Bone Sampling Mode=::NewEnumerator4
    MaskSpawnGroup=false
    Sampled Position Offset=&Add_Vector(
     A=&SelectVectorFromArray(
      Direct Array Index=$Particles.SpawnGroup
      Vector Selection Array=DataInterface() {propertyValues {FloatData [<0 -55 120>, <0 -5 140>, <0 50 160>, <0 108 170>, <0 160 190>, <0 215 205>, <0 270 220>]}}
     )
     B=<0 0 70>
    )
    Skeletal Mesh=DataInterface() {propertyValues {
      DefaultMesh=null
      ExcludeBoneName=root
      OptionalComponentClass=null
      PreviewMesh=&Cube_Queen_Female_Glider
      SkinningMode=SkinOnTheFly
      SoftSourceActor=null
      SourceMode=Default
      UvSetIndex=0
      WholeMeshLOD=-1
      bExcludeBone=false
      bReadDeformedGeometry=true
      bRequireCurrentFrameData=true
      ComponentTags []
      FilteredBones [Root]
      FilteredSockets []
      MeshUserParameter {parameter {name=None, typeDefHandle {registeredTypeIndex=1}, varData []}}
      SamplingRegions []
     }}
   ) {module=SkeletalMeshLocation001}
   &DynamicMaterialParameters(
    Index 0 Param 1=&FloatFromCurve(FloatCurve=DataInterface() {propertyValues {Curve {keys [{interpMode=RCIM_Cubic, tangentMode=RCTM_User, time=0, value=0}, {interpMode=RCIM_Cubic, tangentMode=RCTM_User, time=1, value=1}]}}})
    Index 0 Param 2=40
    Index 0 Param 3=40
    Index 0 Param 4=&RandomRangeFloat(Minimum=0, Maximum=1)
    Module.Param0WriteEnabled=true
    Module.Param1WriteEnabled=true
    Module.Param2WriteEnabled=true
    Module.Param3WriteEnabled=true
   )
   /SetVariables(Particles.Mesh02_DY:Vector4f=&MakeVector4(
     X=&FloatFromCurve(FloatCurve=DataInterface() {propertyValues {Curve {keys [{interpMode=RCIM_Cubic, tangentMode=RCTM_User, time=0, value=0}, {interpMode=RCIM_Cubic, tangentMode=RCTM_User, time=1, value=1}]}}})
     Y=40
     Z=40
     W=0
    )) {isSetParameters=true}
  ]
  rendererClasses [&NiagaraMeshRendererProperties]
  renderers [
   {rendererClass=&NiagaraMeshRendererProperties, rendererIndex=0}
   {rendererClass=&NiagaraMeshRendererProperties, rendererIndex=1}
  ]
 }
 Flame_Left {
  emitterUpdateScript [
   /SetVariables(Emitter.Stairs_Mask_Offset:Vector4f=&MakeVector4(X=0, Y=&FN_FrontEndFloatModifier(Gameplay=-170, Front End=-170), Z=0, W=1)) {isSetParameters=true}
   /SetVariables(Emitter.Stairs_Mask_Offset_Top:Vector4f=&MakeVector4(X=0, Y=&FN_FrontEndFloatModifier(Gameplay=-340, Front End=-340), Z=0, W=1)) {isSetParameters=true}
   &EmitterState(
    Enable Distance Culling=true
    Inactive Response=::NewEnumerator1
    Life Cycle Mode=::NewEnumerator1
    Loop Behavior=::NewEnumerator1
    Loop Duration Mode=::NewEnumerator1
    MaxDistance=7000
    Module.bUseMaxDistance=true
    Module.bUseMinDistance=false
    Reset Age On Awaken=true
    Scalability Mode=::NewEnumerator1
    UseLoopDelay=false
   )
   &SpawnBurst_Instantaneous(Spawn Count=10, Use Spawn Probability=false, UseLoopCountLimit=false)
  ]
  particleSpawnScript [
   &FN_SetSpawnGroup(Number of Spawn Groups=10)
   &InitializeParticle(
    Color=&MakeLinearColorFromVectorAndFloat("Vector (RGB)"=<1 1 1>, "Float (Alpha)"=&FN_FrontEndFloatModifier(Gameplay=3, Front End=6))
    Mesh Scale=&Multiply_Vector(
     A=&RandomRangeVector(Minimum=<4 3 4>, Maximum=<4 3 5>)
     B=&VectorFromFloat(Value=&RandomRangeFloat(Minimum=0.2, Maximum=0.4))
    )
    Mesh Scale Mode=::NewEnumerator4
    UsePositionOffset=false
    Write Material Random=true
   )
   /SetVariables() {isSetParameters=true, module=SetVariables_#1}
   &InitialMeshOrientation(Mesh Orientation Mode=::NewEnumerator2, Rotation=<0 0.02 0>, Use Rotation Vector=true)
   &DynamicMaterialParameters(
    Index 0 Param 1=&RandomRangeFloat(Minimum=0, Maximum=1)
    Index 0 Param 2=&RandomRangeFloat(Minimum=0, Maximum=1)
    Index 0 Param 3=1
    Index 0 Param 4=&RandomRangeFloat(Minimum=0, Maximum=1)
    Index 2 Param 1=&RandomRangeFloat(Minimum=0.8, Maximum=1)
    Module.Param0WriteEnabled=true
    Module.Param0WriteEnabled2=true
    Module.Param1WriteEnabled=true
    Module.Param1WriteEnabled2=true
    Module.Param2WriteEnabled=true
    Module.Param2WriteEnabled2=true
    Module.Param3WriteEnabled=true
    Module.Param3WriteEnabled2=true
    Write Parameter Index 2=true
   ) {module=DynamicMaterialParameters003}
   /SetVariables(Particles.VisibilityTag=&RandomRangeInt(Minimum=0, Maximum=0)) {isSetParameters=true, module=SetVariables_#3}
  ]
  particleUpdateScript [
   &ParticleState(Kill Particles When Lifetime Has Elapsed=false, Let Infinitely Lived Particles Die When Emitter Deactivates=true)
   &SkeletalMeshLocation(
    Bone / Socket Index=$Particles.SpawnGroup
    Bone Sampling Mode=::NewEnumerator4
    MaskSpawnGroup=false
    Sampled Position Offset=&RandomRangeVector(Minimum=<-40 0 0>, Maximum=<-55 0 -5>)
    Skeletal Mesh=$User.SK_Mesh
   )
   &ScaleMeshSize(Scale Factor=&Multiply_VectorByFloat(
     Vector=<1 1 1>
     Float=&MakeCustomFloatFromBool(
      Bool=&SetBoolByFloatComparison(
       A=&VectorLength(Vector=&Subtract_Vector(A=Linked() {linkedVariable {name=Particles.Position, type=NiagaraPosition}}, B=$Particles.Lastposition))
       B=&FN_FrontEndFloatModifier(Gameplay=100, Front End=&Multiply_Float(A=500, B=$Engine.DeltaTime))
      )
      True Float=0
      False Float=1
     )
    )) {module=ScaleMeshSize002}
   /SetVariables(Particles.Lastposition:NiagaraPosition=$Particles.Position) {isSetParameters=true, module=SetVariables_#2}
   &ScaleColor(
    Scale Alpha=&Multiply_Float(
     A=1
     B=&FloatFromCurve(FloatCurve=DataInterface() {propertyValues {Curve {keys [<at 2 0>,<at 3.5 1>]}}}, CurveIndex=$Particles.Age)
    )
    Scale RGB=&Multiply_VectorByFloat(Vector=<1 1 1>, Float=&FN_FrontEndFloatModifier(Gameplay=1, Front End=1.5))
    ScaleA=true
    ScaleRGB=true
   )
   &RegionMask(Offset:Vector3f=<0 -60 0>, Region Falloff=0, Region Radius=220, UseOffset=true, UseRotation=false) {module=RegionMask001}
   &ScaleMeshSize(
    Initial Mesh Scale:Vector3f=$Particles.Scale
    Scale Factor=&Multiply_VectorByFloat(Vector=<1 1 1>, Float=$Particles.RegionMask001.Float)
   )
  ]
  rendererClasses [&NiagaraMeshRendererProperties]
  renderers [{rendererClass=&NiagaraMeshRendererProperties, rendererIndex=0}]
 }
 Flame_Right {
  emitterUpdateScript [
   /SetVariables(Emitter.Stairs_Mask_Offset:Vector4f=&MakeVector4(X=0, Y=&FN_FrontEndFloatModifier(Gameplay=-170, Front End=-170), Z=0, W=1)) {isSetParameters=true}
   /SetVariables(Emitter.Stairs_Mask_Offset_Top:Vector4f=&MakeVector4(X=0, Y=&FN_FrontEndFloatModifier(Gameplay=-340, Front End=-340), Z=0, W=1)) {isSetParameters=true}
   &EmitterState(
    Enable Distance Culling=true
    Inactive Response=::NewEnumerator1
    Life Cycle Mode=::NewEnumerator1
    Loop Behavior=::NewEnumerator1
    Loop Duration Mode=::NewEnumerator1
    MaxDistance=7000
    Module.bUseMaxDistance=true
    Module.bUseMinDistance=false
    Reset Age On Awaken=true
    Scalability Mode=::NewEnumerator1
    UseLoopDelay=false
   )
   &SpawnBurst_Instantaneous(Spawn Count=10, Use Spawn Probability=false, UseLoopCountLimit=false)
  ]
  particleSpawnScript [
   &FN_SetSpawnGroup(Number of Spawn Groups=10)
   &InitializeParticle(
    Color=&MakeLinearColorFromVectorAndFloat("Vector (RGB)"=<1 1 1>, "Float (Alpha)"=&FN_FrontEndFloatModifier(Gameplay=3, Front End=6))
    Mesh Scale=&Multiply_Vector(
     A=&RandomRangeVector(Minimum=<4 3 4>, Maximum=<4 3 5>)
     B=&VectorFromFloat(Value=&RandomRangeFloat(Minimum=0.2, Maximum=0.4))
    )
    Mesh Scale Mode=::NewEnumerator4
    UsePositionOffset=false
    Write Material Random=true
   )
   /SetVariables() {isSetParameters=true, module=SetVariables_#1}
   &InitialMeshOrientation(Mesh Orientation Mode=::NewEnumerator2, Rotation=<0 0 0.5>, Use Rotation Vector=true)
   &InitialMeshOrientation(Mesh Orientation Mode=::NewEnumerator2, Rotation=<0 -0.02 0>, Use Rotation Vector=true) {module=InitialMeshOrientation001}
   &DynamicMaterialParameters(
    Index 0 Param 1=&RandomRangeFloat(Minimum=0, Maximum=1)
    Index 0 Param 2=&RandomRangeFloat(Minimum=0, Maximum=1)
    Index 0 Param 3=1
    Index 0 Param 4=&RandomRangeFloat(Minimum=0, Maximum=1)
    Index 2 Param 1=&RandomRangeFloat(Minimum=0.8, Maximum=1)
    Module.Param0WriteEnabled=true
    Module.Param0WriteEnabled2=true
    Module.Param1WriteEnabled=true
    Module.Param1WriteEnabled2=true
    Module.Param2WriteEnabled=true
    Module.Param2WriteEnabled2=true
    Module.Param3WriteEnabled=true
    Module.Param3WriteEnabled2=true
    Write Parameter Index 2=true
   ) {module=DynamicMaterialParameters002}
   /SetVariables(Particles.VisibilityTag=&RandomRangeInt(Minimum=0, Maximum=0)) {isSetParameters=true, module=SetVariables_#3}
  ]
  particleUpdateScript [
   &ParticleState(Kill Particles When Lifetime Has Elapsed=false, Let Infinitely Lived Particles Die When Emitter Deactivates=true)
   &SkeletalMeshLocation(
    Bone / Socket Index=$Particles.SpawnGroup
    Bone Sampling Mode=::NewEnumerator4
    MaskSpawnGroup=false
    Sampled Position Offset=&RandomRangeVector(Minimum=<40 0 0>, Maximum=<55 0 -5>)
    Skeletal Mesh=$User.SK_Mesh
   ) {module=SkeletalMeshLocation001}
   &ScaleMeshSize(Scale Factor=&Multiply_VectorByFloat(
     Vector=<1 1 1>
     Float=&MakeCustomFloatFromBool(
      Bool=&SetBoolByFloatComparison(
       A=&VectorLength(Vector=&Subtract_Vector(A=Linked() {linkedVariable {name=Particles.Position, type=NiagaraPosition}}, B=$Particles.Lastposition))
       B=&FN_FrontEndFloatModifier(Gameplay=100, Front End=&Multiply_Float(A=500, B=$Engine.DeltaTime))
      )
      True Float=0
      False Float=1
     )
    )) {module=ScaleMeshSize002}
   /SetVariables(Particles.Lastposition:NiagaraPosition=$Particles.Position) {isSetParameters=true, module=SetVariables_#2}
   &ScaleColor(
    Scale Alpha=&Multiply_Float(
     A=1
     B=&FloatFromCurve(FloatCurve=DataInterface() {propertyValues {Curve {keys [<at 2 0>,<at 3.5 1>]}}}, CurveIndex=$Particles.Age)
    )
    Scale RGB=&Multiply_VectorByFloat(Vector=<1 1 1>, Float=&FN_FrontEndFloatModifier(Gameplay=1, Front End=1.5))
    ScaleA=true
    ScaleRGB=true
   )
   &RegionMask(Offset:Vector3f=<0 -60 0>, Region Falloff=0, Region Radius=220, UseOffset=true, UseRotation=false) {module=RegionMask001}
   &ScaleMeshSize(
    Initial Mesh Scale:Vector3f=$Particles.Scale
    Scale Factor=&Multiply_VectorByFloat(Vector=<1 1 1>, Float=$Particles.RegionMask001.Float)
   )
  ]
  rendererClasses [&NiagaraMeshRendererProperties]
  renderers [{rendererClass=&NiagaraMeshRendererProperties, rendererIndex=0}]
 }
 Frontend_smoke {
  emitterUpdateScript [
   &EmitterState(
    Enable Distance Culling=true
    Inactive Response=::NewEnumerator1
    Life Cycle Mode=::NewEnumerator1
    Loop Behavior=::NewEnumerator1
    Loop Delay=1.5
    Loop Duration Mode=::NewEnumerator1
    MaxDistance=17000
    Module.bUseMaxDistance=true
    Module.bUseMinDistance=false
    Scalability Mode=::NewEnumerator1
    UseLoopDelay=true
   )
   &SpawnRate(SpawnRate=&FN_FrontEndFloatModifier(Gameplay=0, Front End=15), Use Spawn Probability=false)
  ]
  particleSpawnScript [
   &FN_SetSpawnGroup(Number of Spawn Groups=10)
   &InitializeParticle(
    Color=&RandomRangeLinearColor(
     Minimum=&LerpLinearColors(StartColor=<rgba 1 1 1 1>, EndColor=<rgba 0.951 0.5336 1 2.2>, LerpFactor=2)
     Maximum=<rgba 0 0.3687 1 2.2>
    )
    Lifetime Max=1.5
    Lifetime Mode=::NewEnumerator1
    Mesh Scale Mode=::NewEnumerator4
    Sprite Rotation Mode=::NewEnumerator2
    Sprite Size Mode=::NewEnumerator1
    Uniform Sprite Size Max=200
    Uniform Sprite Size Min=160
    UsePositionOffset=false
    Write Material Random=true
   )
   /SetVariables() {isSetParameters=true, module=SetVariables_#1}
   /SetVariables(Particles.RandomFrame=&RandomRangeInt(Minimum=0, Maximum=63)) {isSetParameters=true, module=SetVariables_#4}
   &BoxLocation(
    Box Origin:NiagaraPosition=&MakeVector(X=0, Y=0, Z=0)
    Box Size=<140 0 0>
    MaskSpawnGroup=false
    Module.Override Seed=false
    Rotate Box=false
    Surface Only=false
   )
   &DynamicMaterialParameters(
    Index 0 Param 1=&RandomRangeFloat(Minimum=0, Maximum=1, Evaluation Type=::NewEnumerator1)
    Index 0 Param 2=&RandomRangeFloat(Minimum=0, Maximum=1, Evaluation Type=::NewEnumerator1)
    Module.Param0WriteEnabled=true
    Module.Param1WriteEnabled=true
    Module.Param2WriteEnabled=true
    Module.Param3WriteEnabled=true
   ) {module=DynamicMaterialParameters002}
  ]
  particleUpdateScript [
   &ParticleState()
   &SkeletalMeshLocation(
    Bone / Socket Index=$Particles.SpawnGroup
    Bone Sampling Mode=::NewEnumerator4
    MaskSpawnGroup=false
    Sampled Orientation Coordinate Space=::Local
    Sampled Position Offset=&Add_Vector(A=Linked() {linkedVariable {name=Particles.Initial.Position, type=NiagaraPosition}}, B=<0 0 0>)
    Skeletal Mesh=$User.SK_Mesh
   )
   &ScaleSpriteSize(
    Scale Sprite Size Mode=::NewEnumerator0
    Uniform Scale Factor=&Multiply_Float(
     A=&FloatFromCurve(FloatCurve=DataInterface() {propertyValues {Curve {keys [
          {arriveTangent=1.1933, interpMode=RCIM_Cubic, leaveTangent=1.1933, tangentMode=RCTM_User, time=0, value=0.6}
          {arriveTangent=-0.0405, interpMode=RCIM_Cubic, leaveTangent=-0.0405, tangentMode=RCTM_User, time=1, value=1}
         ]}}})
     B=&MakeCustomFloatFromBool(
      Bool=&SetBoolByFloatComparison(
       A=&VectorLength(Vector=&Subtract_Vector(
         A=&ConvertPositionToVector(Input Position=$Particles.Position, Convert To Absolute Large World Vector=false)
         B=$Particles.Lastposition
        ))
       B=20
      )
      True Float=0
      False Float=&FN_FrontEndFloatModifier(Gameplay=1, Front End=0.85)
     )
    )
   ) {module=ScaleSpriteSize001}
   /SetVariables(Particles.Lastposition:Vector3f=&ConvertPositionToVector(Input Position=$Particles.Position, Convert To Absolute Large World Vector=false)) {isSetParameters=true, module=SetVariables_#2}
   &SpriteRotationRate(
    Delta Time:NiagaraFloat=$Particles.NormalizedAge
    Rotation Rate=&RandomRangeFloat(Minimum=-2, Maximum=2)
   )
   &RegionMask(Offset:Vector3f=<0 120 0>, Region Radius=230, UseOffset=true, UseRotation=false)
   &ScaleColor(
    Scale Alpha=&Multiply_Float(
     A=&Multiply_Float(
      A=&FloatFromCurve(
       FloatCurve=DataInterface() {propertyValues {Curve {keys [
           {interpMode=RCIM_Cubic, tangentMode=RCTM_User, time=0, value=0}
           {interpMode=RCIM_Cubic, tangentMode=RCTM_User, time=0.5, value=1}
           {tangentMode=RCTM_User, time=1, value=0}
          ]}}}
       Scale Curve=&RandomRangeFloat(Minimum=0.8, Maximum=1.5)
      )
      B=&Multiply_Float(
       A=&FNTimeOfDayScalar(Night Scale=0.5, Morning Scale=1, Evening Scale=1)
       B=$Particles.RegionMask.Float
      )
     )
     B=&FN_FrontEndFloatModifier(Gameplay=1, Front End=0.17)
    )
    Scale RGB=&Multiply_VectorByFloat(
     Vector=<1 1 1>
     Float=&FN_FrontEndFloatModifier(Gameplay=&FNTimeOfDayScalar(Night Scale=0.5, Morning Scale=0.1, Evening Scale=0.2), Front End=3)
    )
    ScaleA=true
    ScaleRGB=true
   ) {module=ScaleColor002}
   &RegionMask(Offset:Vector3f=<0 120 0>, Region Falloff=0, Region Radius=250, UseOffset=true, UseRotation=false) {module=RegionMask001}
   &ScaleSpriteSize(
    Initial Sprite Size:Vector2f=$Particles.SpriteSize
    Scale Sprite Size Mode=::NewEnumerator0
    Uniform Scale Factor=$Particles.RegionMask001.Float
   )
  ]
  rendererClasses [&NiagaraSpriteRendererProperties]
  renderers [{rendererClass=&NiagaraSpriteRendererProperties}]
 }
 Only_Frontend_Trail {
  emitterUpdateScript [
   &EmitterState(Life Cycle Mode=::NewEnumerator1, Loop Behavior=::NewEnumerator1, Loop Delay=0.5, Loop Duration=1.7, UseLoopDelay=true)
   &SpawnRate(SpawnRate=&FN_FrontEndFloatModifier(Gameplay=0, Front End=30), Use Spawn Probability=false)
  ]
  particleSpawnScript [
   &FN_SetSpawnGroup(Number of Spawn Groups=1)
   &InitializeParticle(
    Color=&RandomRangeLinearColor(
     Minimum=&LerpLinearColors(StartColor=<rgba 1 1 1 1>, EndColor=<rgba 0.951 0.5336 1 2>, LerpFactor=2)
     Maximum=<rgba 0 0.3687 1 2>
    )
    Lifetime Max=0.6
    Lifetime Min=0.5
    Lifetime Mode=::NewEnumerator1
    Mesh Scale Mode=::NewEnumerator4
    Sprite Rotation Mode=::NewEnumerator2
    Sprite Size Mode=::NewEnumerator1
    Uniform Sprite Size Max=120
    Uniform Sprite Size Min=100
    UsePositionOffset=false
    Write Material Random=true
   ) {module=InitializeParticle001}
   &SkeletalMeshLocation(
    Bone Sampling Mode=::NewEnumerator4
    MaskSpawnGroup=false
    Skeletal Mesh=DataInterface() {propertyValues {
      DefaultMesh=null
      ExcludeBoneName=root
      OptionalComponentClass=null
      PreviewMesh=&Cube_Queen_Female_Glider
      SkinningMode=SkinOnTheFly
      SoftSourceActor=null
      SourceMode=Default
      UvSetIndex=0
      WholeMeshLOD=-1
      bExcludeBone=false
      bReadDeformedGeometry=true
      bRequireCurrentFrameData=true
      ComponentTags []
      FilteredBones [JNT_Platform_03]
      FilteredSockets []
      MeshUserParameter {parameter {name=None, typeDefHandle {registeredTypeIndex=1}, varData []}}
      SamplingRegions []
     }}
   ) {module=SkeletalMeshLocation001}
   &SphereLocation(
    MaskSpawnGroup=false
    Override Seed=false
    Sphere Radius=5
    Surface Only=false
    UseNonUniformScale=false
    UseOffset=false
    UseSphereOrientation=false
   )
   &DynamicMaterialParameters(
    Index 0 Param 1=&RandomRangeFloat(Minimum=0, Maximum=1, Evaluation Type=::NewEnumerator1)
    Index 0 Param 2=&RandomRangeFloat(Minimum=0, Maximum=1, Evaluation Type=::NewEnumerator1)
    Module.Param0WriteEnabled=true
    Module.Param1WriteEnabled=true
    Module.Param2WriteEnabled=true
    Module.Param3WriteEnabled=true
   ) {module=DynamicMaterialParameters002}
  ]
  particleUpdateScript [
   &ParticleState()
   &SolveForcesAndVelocity(Clamp Velocity=false, Limit Acceleration=false)
   &UpdateMeshOrientation(
    Rotation Rate=&FloatFromCurve(
     FloatCurve=DataInterface() {propertyValues {Curve {keys [
         {arriveTangent=-2.1159, interpMode=RCIM_Cubic, leaveTangent=-2.1159, tangentMode=RCTM_User, time=0, value=1}
         {interpMode=RCIM_Cubic, tangentMode=RCTM_Break, time=1, value=0.2}
        ]}}}
     Scale Curve=&RandomRangeFloat(Minimum=0.7, Maximum=1.5)
    )
    Rotation Vector=&RandomRangeVector(Minimum=<-1 -1 -1>, Maximum=<1 1 1>)
   )
   &ScaleSpriteSize(Uniform Curve Sprite Scale=DataInterface() {propertyValues {Curve {keys [{interpMode=RCIM_Cubic, tangentMode=RCTM_User, time=0, value=1}, {interpMode=RCIM_Cubic, tangentMode=RCTM_User, time=1, value=0.5}]}}})
   &ScaleColor(
    Scale Alpha=&Multiply_Float(
     A=&Multiply_Float(
      A=&FloatFromCurve(
       FloatCurve=DataInterface() {propertyValues {Curve {keys [
           {interpMode=RCIM_Cubic, tangentMode=RCTM_User, time=0, value=0}
           {interpMode=RCIM_Cubic, tangentMode=RCTM_User, time=0.5, value=1}
           {tangentMode=RCTM_User, time=1, value=0}
          ]}}}
       Scale Curve=&RandomRangeFloat(Minimum=0.8, Maximum=1.5)
      )
      B=&Multiply_Float(
       A=&FNTimeOfDayScalar(Night Scale=0.6, Morning Scale=0.15, Evening Scale=0.3)
       B=&FloatFromCurve(FloatCurve=DataInterface() {propertyValues {Curve {keys [<at 0 0>,<at 0.2 0.8>,<at 1 0>]}}}, CurveIndex=$Particles.Age)
      )
     )
     B=0.1
    )
    Scale RGB=&Multiply_VectorByFloat(Vector=<1 1 1>, Float=3)
    ScaleA=true
    ScaleRGB=true
   ) {module=ScaleColor002}
  ]
  rendererClasses [&NiagaraSpriteRendererProperties]
  renderers [{rendererClass=&NiagaraSpriteRendererProperties}]
 }
 Range_cube {
  emitterUpdateScript [
   &EmitterState(
    Enable Distance Culling=true
    Inactive Response=::NewEnumerator1
    Life Cycle Mode=::NewEnumerator1
    Loop Behavior=::NewEnumerator1
    Loop Delay=1.5
    Loop Duration Mode=::NewEnumerator1
    MaxDistance=3000
    Module.bUseMaxDistance=true
    Module.bUseMinDistance=false
    Scalability Mode=::NewEnumerator1
    UseLoopDelay=true
   )
   &SpawnRate(SpawnRate=&FN_FrontEndFloatModifier(Gameplay=50, Front End=30), Use Spawn Probability=false)
  ]
  particleSpawnScript [
   &FN_SetSpawnGroup(Number of Spawn Groups=10)
   &InitializeParticle(
    Color=<rgba 1.5 1.5 1.5 0.8>
    Lifetime Mode=::NewEnumerator1
    Mesh Scale Mode=::NewEnumerator1
    Mesh Uniform Scale Max=5
    Mesh Uniform Scale Min=3
    Sprite Rotation Mode=::NewEnumerator2
    Sprite Size Mode=::NewEnumerator1
    Uniform Sprite Size Max=200
    Uniform Sprite Size Min=160
    UsePositionOffset=false
    Write Material Random=true
   )
   /SetVariables() {isSetParameters=true, module=SetVariables_#1}
   /SetVariables(Particles.RandomFrame=&RandomRangeInt(Minimum=0, Maximum=63)) {isSetParameters=true, module=SetVariables_#4}
   &BoxLocation(
    Box Origin:NiagaraPosition=&MakeVector(X=0, Y=0, Z=0)
    Box Size=<140 0 0>
    MaskSpawnGroup=false
    Module.Override Seed=false
    Rotate Box=false
    Surface Only=false
   )
   &DynamicMaterialParameters(
    Index 0 Param 1=&RandomRangeFloat(Minimum=0, Maximum=1, Evaluation Type=::NewEnumerator1)
    Index 0 Param 2=&RandomRangeFloat(Minimum=0, Maximum=1, Evaluation Type=::NewEnumerator1)
    Module.Param0WriteEnabled=true
    Module.Param1WriteEnabled=true
    Module.Param2WriteEnabled=true
    Module.Param3WriteEnabled=true
   ) {module=DynamicMaterialParameters002}
   &InitialMeshOrientation(Use Rotation Vector=false)
   &DynamicMaterialParameters(
    Index 1 Param 1=-1
    Index 1 Param 3=1
    Module.Param0WriteEnabled=true
    Module.Param0WriteEnabled1=true
    Module.Param1WriteEnabled=true
    Module.Param1WriteEnabled1=true
    Module.Param2WriteEnabled=true
    Module.Param2WriteEnabled1=true
    Module.Param3WriteEnabled=true
    Module.Param3WriteEnabled1=true
    Write Parameter Index 1=true
   ) {module=DynamicMaterialParameters003}
  ]
  particleUpdateScript [
   &ParticleState()
   &SkeletalMeshLocation(
    Bone / Socket Index=$Particles.SpawnGroup
    Bone Sampling Mode=::NewEnumerator4
    MaskSpawnGroup=false
    Sampled Orientation Coordinate Space=::Local
    Sampled Position Offset=&Add_Vector(
     A=Linked() {linkedVariable {name=Particles.Initial.Position, type=NiagaraPosition}}
     B=&MakeVector(
      X=0
      Y=&Lerp_Float(A=0, B=50, Alpha=$Particles.NormalizedAge)
      Z=&Lerp_Float(A=0, B=100, Alpha=$Particles.NormalizedAge)
     )
    )
    Skeletal Mesh=$User.SK_Mesh
   )
   &ScaleMeshSize(Scale Factor=&Multiply_VectorByFloat(
     Vector=&Multiply_VectorByFloat(
      Vector=<1 1 1>
      Float=&FloatFromCurve(
       FloatCurve=DataInterface() {propertyValues {Curve {keys [
           {interpMode=RCIM_Cubic, tangentMode=RCTM_User, time=0, value=0}
           <at 0.3 0.9 RCIM_Cubic>
           {interpMode=RCIM_Cubic, tangentMode=RCTM_User, time=1, value=0}
          ]}}}
       Scale Curve=&MakeCustomFloatFromBool(Bool=$User.IsSwing, True Float=0, False Float=1)
      )
     )
     Float=&MakeCustomFloatFromBool(
      Bool=&SetBoolByFloatComparison(
       A=&VectorLength(Vector=&Subtract_Vector(A=Linked() {linkedVariable {name=Particles.Position, type=NiagaraPosition}}, B=$Particles.Lastposition))
       B=100
      )
      True Float=0
      False Float=&FN_FrontEndFloatModifier(Gameplay=1, Front End=1.4)
     )
    )) {module=ScaleMeshSize002}
   /SetVariables(Particles.Lastposition:NiagaraPosition=$Particles.SkeletalMeshLocation.SampledPosition) {isSetParameters=true, module=SetVariables_#2}
   &SpriteRotationRate(
    Delta Time:NiagaraFloat=$Particles.NormalizedAge
    Rotation Rate=&RandomRangeFloat(Minimum=-2, Maximum=2)
   )
   &RegionMask(
    Offset:Vector3f=&FN_FrontEndVector3Modifier(Gameplay=<0 250 0>, Front End=<0 100 0>)
    Region Radius=230
    UseOffset=true
    UseRotation=false
   )
   &ScaleColor(
    Scale Alpha=&Multiply_Float(
     A=&Multiply_Float(
      A=&FloatFromCurve(
       FloatCurve=DataInterface() {propertyValues {Curve {keys [
           {interpMode=RCIM_Cubic, tangentMode=RCTM_User, time=0, value=0}
           {interpMode=RCIM_Cubic, tangentMode=RCTM_User, time=0.5, value=1}
           {tangentMode=RCTM_User, time=1, value=0}
          ]}}}
       Scale Curve=&RandomRangeFloat(Minimum=0.8, Maximum=1.5)
      )
      B=&Multiply_Float(
       A=&FNTimeOfDayScalar(Night Scale=0.6, Morning Scale=0.15, Evening Scale=0.3)
       B=$Particles.RegionMask.Float
      )
     )
     B=1
    )
    ScaleA=true
    ScaleRGB=true
   ) {module=ScaleColor002}
  ]
  rendererClasses [&NiagaraMeshRendererProperties]
  renderers [{rendererClass=&NiagaraMeshRendererProperties, rendererIndex=0}]
 }
 Smoke {
  emitterUpdateScript [
   &EmitterState(
    Enable Distance Culling=true
    Inactive Response=::NewEnumerator1
    Life Cycle Mode=::NewEnumerator1
    Loop Behavior=::NewEnumerator1
    Loop Delay=1.5
    Loop Duration Mode=::NewEnumerator1
    MaxDistance=17000
    Module.bUseMaxDistance=true
    Module.bUseMinDistance=false
    Scalability Mode=::NewEnumerator1
    UseLoopDelay=true
   )
   &SpawnRate(SpawnRate=&FN_FrontEndFloatModifier(Gameplay=15, Front End=0), Use Spawn Probability=false)
  ]
  particleSpawnScript [
   &FN_SetSpawnGroup(Number of Spawn Groups=10)
   &InitializeParticle(
    Color=&RandomRangeLinearColor(
     Minimum=&LerpLinearColors(StartColor=<rgba 1 1 1 1>, EndColor=<rgba 0.951 0.5336 1 2.2>, LerpFactor=2)
     Maximum=<rgba 0 0.3687 1 2.2>
    )
    Lifetime Max=1.5
    Lifetime Mode=::NewEnumerator1
    Mesh Scale Mode=::NewEnumerator4
    Sprite Rotation Mode=::NewEnumerator2
    Sprite Size Mode=::NewEnumerator1
    Uniform Sprite Size Max=200
    Uniform Sprite Size Min=160
    UsePositionOffset=false
    Write Material Random=true
   )
   /SetVariables() {isSetParameters=true, module=SetVariables_#1}
   /SetVariables(Particles.RandomFrame=&RandomRangeInt(Minimum=0, Maximum=63)) {isSetParameters=true, module=SetVariables_#4}
   &BoxLocation(
    Box Origin:NiagaraPosition=&MakeVector(X=0, Y=0, Z=0)
    Box Size=<140 0 0>
    MaskSpawnGroup=false
    Module.Override Seed=false
    Rotate Box=false
    Surface Only=false
   )
   &DynamicMaterialParameters(
    Index 0 Param 1=&RandomRangeFloat(Minimum=0, Maximum=1, Evaluation Type=::NewEnumerator1)
    Index 0 Param 2=&RandomRangeFloat(Minimum=0, Maximum=1, Evaluation Type=::NewEnumerator1)
    Module.Param0WriteEnabled=true
    Module.Param1WriteEnabled=true
    Module.Param2WriteEnabled=true
    Module.Param3WriteEnabled=true
   ) {module=DynamicMaterialParameters002}
  ]
  particleUpdateScript [
   &ParticleState()
   &SkeletalMeshLocation(
    Bone / Socket Index=$Particles.SpawnGroup
    Bone Sampling Mode=::NewEnumerator4
    MaskSpawnGroup=false
    Sampled Orientation Coordinate Space=::Local
    Sampled Position Offset=&Add_Vector(A=Linked() {linkedVariable {name=Particles.Initial.Position, type=NiagaraPosition}}, B=<0 0 0>)
    Skeletal Mesh=$User.SK_Mesh
   )
   &ScaleSpriteSize(
    Scale Sprite Size Mode=::NewEnumerator0
    Uniform Scale Factor=&Multiply_Float(
     A=&FloatFromCurve(FloatCurve=DataInterface() {propertyValues {Curve {keys [
          {arriveTangent=1.1933, interpMode=RCIM_Cubic, leaveTangent=1.1933, tangentMode=RCTM_User, time=0, value=0.6}
          {arriveTangent=-0.0405, interpMode=RCIM_Cubic, leaveTangent=-0.0405, tangentMode=RCTM_User, time=1, value=1}
         ]}}})
     B=&MakeCustomFloatFromBool(
      Bool=&SetBoolByFloatComparison(
       A=&VectorLength(Vector=&Subtract_Vector(A=Linked() {linkedVariable {name=Particles.Position, type=NiagaraPosition}}, B=$Particles.Lastposition))
       B=100
      )
      True Float=0
      False Float=&FN_FrontEndFloatModifier(Gameplay=1, Front End=0.85)
     )
    )
   ) {module=ScaleSpriteSize001}
   /SetVariables(Particles.Lastposition:NiagaraPosition=$Particles.SkeletalMeshLocation.SampledPosition) {isSetParameters=true, module=SetVariables_#2}
   &SpriteRotationRate(
    Delta Time:NiagaraFloat=$Particles.NormalizedAge
    Rotation Rate=&RandomRangeFloat(Minimum=-2, Maximum=2)
   )
   &RegionMask(Offset:Vector3f=<0 120 0>, Region Radius=230, UseOffset=true, UseRotation=false)
   &ScaleColor(
    Scale Alpha=&Multiply_Float(
     A=&Multiply_Float(
      A=&FloatFromCurve(
       FloatCurve=DataInterface() {propertyValues {Curve {keys [
           {interpMode=RCIM_Cubic, tangentMode=RCTM_User, time=0, value=0}
           {interpMode=RCIM_Cubic, tangentMode=RCTM_User, time=0.5, value=1}
           {tangentMode=RCTM_User, time=1, value=0}
          ]}}}
       Scale Curve=&RandomRangeFloat(Minimum=0.8, Maximum=1.5)
      )
      B=&Multiply_Float(
       A=&FNTimeOfDayScalar(Night Scale=0.5, Morning Scale=1, Evening Scale=1)
       B=$Particles.RegionMask.Float
      )
     )
     B=&FN_FrontEndFloatModifier(Gameplay=1, Front End=0.17)
    )
    Scale RGB=&Multiply_VectorByFloat(
     Vector=<1 1 1>
     Float=&FN_FrontEndFloatModifier(Gameplay=&FNTimeOfDayScalar(Night Scale=0.5, Morning Scale=0.1, Evening Scale=0.2), Front End=3)
    )
    ScaleA=true
    ScaleRGB=true
   ) {module=ScaleColor002}
   &RegionMask(Offset:Vector3f=<0 120 0>, Region Falloff=0, Region Radius=250, UseOffset=true, UseRotation=false) {module=RegionMask001}
   &ScaleSpriteSize(
    Initial Sprite Size:Vector2f=$Particles.SpriteSize
    Scale Sprite Size Mode=::NewEnumerator0
    Uniform Scale Factor=$Particles.RegionMask001.Float
   )
  ]
  rendererClasses [&NiagaraSpriteRendererProperties]
  renderers [{rendererClass=&NiagaraSpriteRendererProperties}]
 }
 Stair_fragments {
  emitterUpdateScript [
   /SetVariables(Emitter.Dissipate_Stairs_Mask_Offset:Vector4f=&MakeVector4(X=0, Y=&FN_FrontEndFloatModifier(Gameplay=-100, Front End=-100), Z=0, W=1)) {isSetParameters=true}
   /SetVariables(Emitter.Dissipate_Stairs_Mask_Offset_Top:Vector4f=&MakeVector4(X=0, Y=&FN_FrontEndFloatModifier(Gameplay=-250, Front End=-250), Z=0, W=1)) {isSetParameters=true}
   &EmitterState(
    Inactive Response=::NewEnumerator1
    Life Cycle Mode=::NewEnumerator1
    Loop Behavior=::NewEnumerator1
    Loop Duration Mode=::NewEnumerator1
    UseLoopDelay=false
   )
   &SpawnBurst_Instantaneous(Spawn Count=10, Use Spawn Probability=false, UseLoopCountLimit=false)
  ]
  particleSpawnScript [
   &FN_SetSpawnGroup(Number of Spawn Groups=10)
   &InitializeParticle(
    Color=&MakeLinearColorFromVectorAndFloat("Vector (RGB)"=<1 1 1>, "Float (Alpha)"=&FN_FrontEndFloatModifier(Gameplay=0.3, Front End=1))
    Mesh Scale=&Multiply_VectorByFloat(Vector=<1 1 1>, Float=&RandomRangeFloat(Minimum=0.9, Maximum=1.2))
    Mesh Scale Mode=::NewEnumerator4
    UsePositionOffset=false
    Write Material Random=true
   )
   /SetVariables() {isSetParameters=true, module=SetVariables_#1}
   /SetVariables(Particles.MeshIndex=&RandomRangeInt(Minimum=0, Maximum=0)) {enabled=false, isSetParameters=true}
  ]
  particleUpdateScript [
   &ParticleState(Kill Particles When Lifetime Has Elapsed=false, Let Infinitely Lived Particles Die When Emitter Deactivates=true)
   &SkeletalMeshLocation(
    Bone / Socket Index=$Particles.SpawnGroup
    Bone Sampling Mode=::NewEnumerator4
    MaskSpawnGroup=false
    Sampled Position Offset=<0 20 10>
    Skeletal Mesh=$User.SK_Mesh
   )
   &ScaleMeshSize(Scale Factor=&Multiply_VectorByFloat(
     Vector=<1 1 1>
     Float=&MakeCustomFloatFromBool(
      Bool=&SetBoolByFloatComparison(
       A=&VectorLength(Vector=&Subtract_Vector(A=Linked() {linkedVariable {name=Particles.Position, type=NiagaraPosition}}, B=$Particles.Lastposition))
       B=&FN_FrontEndFloatModifier(Gameplay=100, Front End=&Multiply_Float(A=500, B=$Engine.DeltaTime))
      )
      True Float=0
      False Float=1
     )
    )) {module=ScaleMeshSize003}
   /SetVariables(Particles.Lastposition:NiagaraPosition=$Particles.Position) {isSetParameters=true, module=SetVariables_#2}
   &ScaleColor(
    Scale Alpha=&FloatFromCurve(FloatCurve=DataInterface() {propertyValues {Curve {keys [<at 1.09 0>,<at 1.1 1>]}}}, CurveIndex=$Particles.Age)
    Scale RGB=&Multiply_VectorByFloat(Vector=<1 1 1>, Float=&FN_FrontEndFloatModifier(Gameplay=1, Front End=1.5))
    ScaleA=true
    ScaleRGB=true
   ) {module=ScaleColor001}
   &ScaleColor(Scale Alpha=&Multiply_Float(A=1, B=1), ScaleA=true, ScaleRGB=true) {enabled=false}
   &DynamicMaterialParameters(
    Index 0 Param 2=1
    Index 0 Param 3=0.5
    Index 1 Param 1=1
    Index 1 Param 2=0.5
    Module.Param0WriteEnabled=true
    Module.Param0WriteEnabled1=true
    Module.Param1WriteEnabled=true
    Module.Param1WriteEnabled1=true
    Module.Param2WriteEnabled=true
    Module.Param2WriteEnabled1=true
    Module.Param3WriteEnabled=true
    Module.Param3WriteEnabled1=true
    Write Parameter Index 1=true
   ) {module=DynamicMaterialParameters001}
   &RegionMask(Offset:Vector3f=<0 250 0>, Region Falloff=0, Region Radius=150, UseOffset=true, UseRotation=false)
   &ScaleMeshSize(
    Initial Mesh Scale:Vector3f=$Particles.Scale
    Scale Factor=&Multiply_VectorByFloat(Vector=<1 1 1>, Float=$Particles.RegionMask.Float)
   )
  ]
  rendererClasses [&NiagaraMeshRendererProperties]
  renderers [{rendererClass=&NiagaraMeshRendererProperties, rendererIndex=0}]
 }
 The_first_half_of_the_stairs {
  emitterUpdateScript [
   /SetVariables(Emitter.Stairs_Mask_Offset:Vector4f=&MakeVector4(X=0, Y=&FN_FrontEndFloatModifier(Gameplay=-100, Front End=-100), Z=0, W=1)) {isSetParameters=true}
   /SetVariables(Emitter.Stairs_Mask_Offset_Top:Vector4f=&MakeVector4(X=0, Y=&FN_FrontEndFloatModifier(Gameplay=-370, Front End=-390), Z=0, W=1)) {isSetParameters=true}
   &EmitterState(
    Inactive Response=::NewEnumerator1
    Life Cycle Mode=::NewEnumerator1
    Loop Behavior=::NewEnumerator1
    Loop Duration Mode=::NewEnumerator1
    UseLoopDelay=false
   )
   &SpawnBurst_Instantaneous(Spawn Count=10, Use Spawn Probability=false, UseLoopCountLimit=false)
  ]
  particleSpawnScript [
   &FN_SetSpawnGroup(Number of Spawn Groups=10)
   &InitializeParticle(Mesh Scale Mode=::NewEnumerator4, UsePositionOffset=false, Write Material Random=true)
   /SetVariables() {isSetParameters=true, module=SetVariables_#1}
  ]
  particleUpdateScript [
   &ParticleState(Kill Particles When Lifetime Has Elapsed=false, Let Infinitely Lived Particles Die When Emitter Deactivates=true)
   &SkeletalMeshLocation(Bone / Socket Index=$Particles.SpawnGroup, Bone Sampling Mode=::NewEnumerator4, MaskSpawnGroup=false, Skeletal Mesh=$User.SK_Mesh)
   &ScaleMeshSize(Scale Factor=&Multiply_VectorByFloat(
     Vector=<1 1 1>
     Float=&MakeCustomFloatFromBool(
      Bool=&SetBoolByFloatComparison(
       A=&VectorLength(Vector=&Subtract_Vector(A=Linked() {linkedVariable {name=Particles.Position, type=NiagaraPosition}}, B=$Particles.Lastposition))
       B=&FN_FrontEndFloatModifier(Gameplay=100, Front End=&Multiply_Float(A=500, B=$Engine.DeltaTime))
      )
      True Float=0
      False Float=1
     )
    ))
   /SetVariables(Particles.Lastposition:NiagaraPosition=$Particles.Position) {isSetParameters=true, module=SetVariables_#2}
   &DynamicMaterialParameters(
    Index 0 Param 1=&SelectFloatFromArray(
     Direct Array Index=$Particles.SpawnGroup
     Float Selection Array=DataInterface() {propertyValues {FloatData [0.3,0.7,0,0,0,0,0.6,0.3,0.5,0.7]}}
    )
    Index 0 Param 2=&SelectFloatFromArray(
     Direct Array Index=$Particles.SpawnGroup
     Float Selection Array=DataInterface() {propertyValues {FloatData [0.7,0.7,0.8,0.8,1.1,0.89,1,1,1,1]}}
    )
    Index 0 Param 3=&SelectFloatFromArray(
     Direct Array Index=$Particles.SpawnGroup
     Float Selection Array=DataInterface() {propertyValues {FloatData [0.8,1.1,0.85,0.8,0.9,1.1,1,1,1,1]}}
    )
    Index 1 Param 1=&SelectFloatFromArray(Direct Array Index=$Particles.SpawnGroup, Float Selection Array=DataInterface() {propertyValues {FloatData [0.5,0.7,1.2,1,1,1.2,1,1,1,1]}})
    Index 1 Param 2=&SelectFloatFromArray(Direct Array Index=$Particles.SpawnGroup, Float Selection Array=DataInterface() {propertyValues {FloatData [0.5,1.2,1.5,1,1,0.7,1,1,1,1]}})
    Module.Param0WriteEnabled=true
    Module.Param0WriteEnabled1=true
    Module.Param1WriteEnabled=true
    Module.Param1WriteEnabled1=true
    Module.Param2WriteEnabled=true
    Module.Param2WriteEnabled1=true
    Module.Param3WriteEnabled=true
    Module.Param3WriteEnabled1=true
    Write Parameter Index 1=true
   ) {module=DynamicMaterialParameters001}
   &ScaleColor(
    Scale Alpha=&FloatFromCurve(FloatCurve=DataInterface() {propertyValues {Curve {keys [<at 1.09 0>,<at 1.1 1>]}}}, CurveIndex=$Particles.Age)
    ScaleA=true
    ScaleRGB=true
   ) {module=ScaleColor001}
   &RegionMask(Offset:Vector3f=<0 -60 0>, Region Falloff=0.2, Region Radius=160, UseOffset=true, UseRotation=false)
   &ScaleColor(
    Scale Alpha=&Multiply_Float(A=$Particles.RegionMask.Float, B=1)
    Scale RGB=&Multiply_VectorByFloat(Vector=<1 1 1>, Float=&FN_FrontEndFloatModifier(Gameplay=1, Front End=1.5))
    ScaleA=true
    ScaleRGB=true
   )
   &RegionMask(Offset:Vector3f=<0 -60 0>, Region Falloff=0, Region Radius=160, UseOffset=true, UseRotation=false) {module=RegionMask001}
   &ScaleMeshSize(
    Initial Mesh Scale:Vector3f=$Particles.Scale
    Scale Factor=&Multiply_VectorByFloat(Vector=<1 1 1>, Float=$Particles.RegionMask001.Float)
   ) {module=ScaleMeshSize001}
  ]
  rendererClasses [&NiagaraMeshRendererProperties]
  renderers [{rendererClass=&NiagaraMeshRendererProperties, rendererIndex=0}]
 }
 The_second_half_of_the_stairs {
  emitterUpdateScript [
   /SetVariables(Emitter.Stairs_Mask_Offset:Vector4f=&MakeVector4(X=0, Y=&FN_FrontEndFloatModifier(Gameplay=-150, Front End=-150), Z=0, W=1)) {isSetParameters=true}
   /SetVariables(Emitter.Stairs_Mask_Offset_Top:Vector4f=&MakeVector4(X=0, Y=&FN_FrontEndFloatModifier(Gameplay=-400, Front End=-400), Z=0, W=1)) {isSetParameters=true}
   &EmitterState(
    Inactive Response=::NewEnumerator1
    Life Cycle Mode=::NewEnumerator1
    Loop Behavior=::NewEnumerator1
    Loop Duration Mode=::NewEnumerator1
    UseLoopDelay=false
   )
   &SpawnBurst_Instantaneous(Spawn Count=10, Use Spawn Probability=false, UseLoopCountLimit=false)
  ]
  particleSpawnScript [
   &FN_SetSpawnGroup(Number of Spawn Groups=10)
   &InitializeParticle(
    Color=&MakeLinearColorFromVectorAndFloat("Vector (RGB)"=<1 1 1>, "Float (Alpha)"=&FN_FrontEndFloatModifier(Gameplay=0.65, Front End=1))
    Mesh Scale Mode=::NewEnumerator4
    UsePositionOffset=false
    Write Material Random=true
   )
   /SetVariables() {isSetParameters=true, module=SetVariables_#1}
  ]
  particleUpdateScript [
   &ParticleState(Kill Particles When Lifetime Has Elapsed=false, Let Infinitely Lived Particles Die When Emitter Deactivates=true)
   &SkeletalMeshLocation(Bone / Socket Index=$Particles.SpawnGroup, Bone Sampling Mode=::NewEnumerator4, MaskSpawnGroup=false, Skeletal Mesh=$User.SK_Mesh)
   &ScaleMeshSize(Scale Factor=&Multiply_VectorByFloat(
     Vector=<1 1 1>
     Float=&MakeCustomFloatFromBool(
      Bool=&SetBoolByFloatComparison(
       A=&VectorLength(Vector=&Subtract_Vector(A=Linked() {linkedVariable {name=Particles.Position, type=NiagaraPosition}}, B=$Particles.Lastposition))
       B=&FN_FrontEndFloatModifier(Gameplay=100, Front End=&Multiply_Float(A=500, B=$Engine.DeltaTime))
      )
      True Float=0
      False Float=1
     )
    )) {module=ScaleMeshSize002}
   /SetVariables(Particles.Lastposition:NiagaraPosition=$Particles.Position) {isSetParameters=true, module=SetVariables_#2}
   &DynamicMaterialParameters(
    Index 0 Param 1=&SelectFloatFromArray(
     Direct Array Index=$Particles.SpawnGroup
     Float Selection Array=DataInterface() {propertyValues {FloatData [0.3,0.7,0,0,0,0,0.6,0.3,0.5,0.7]}}
    )
    Index 0 Param 2=&SelectFloatFromArray(
     Direct Array Index=$Particles.SpawnGroup
     Float Selection Array=DataInterface() {propertyValues {FloatData [0.7,0.7,0.8,0.8,1.1,0.89,1,1,1,1]}}
    )
    Index 0 Param 3=&SelectFloatFromArray(
     Direct Array Index=$Particles.SpawnGroup
     Float Selection Array=DataInterface() {propertyValues {FloatData [0.8,1.1,0.85,0.8,0.9,1.1,1,1,1,1]}}
    )
    Index 1 Param 1=&SelectFloatFromArray(Direct Array Index=$Particles.SpawnGroup, Float Selection Array=DataInterface() {propertyValues {FloatData [0.5,0.7,1.2,1,1,1.2,1,1,1,1]}})
    Index 1 Param 2=&SelectFloatFromArray(Direct Array Index=$Particles.SpawnGroup, Float Selection Array=DataInterface() {propertyValues {FloatData [0.5,1.2,1.5,1,1,0.7,1,1,1,1]}})
    Module.Param0WriteEnabled=true
    Module.Param0WriteEnabled1=true
    Module.Param1WriteEnabled=true
    Module.Param1WriteEnabled1=true
    Module.Param2WriteEnabled=true
    Module.Param2WriteEnabled1=true
    Module.Param3WriteEnabled=true
    Module.Param3WriteEnabled1=true
    Write Parameter Index 1=true
   ) {module=DynamicMaterialParameters001}
   &ScaleColor(
    Scale Alpha=&FloatFromCurve(FloatCurve=DataInterface() {propertyValues {Curve {keys [<at 1.09 0>,<at 1.1 1>]}}}, CurveIndex=$Particles.Age)
    ScaleA=true
    ScaleRGB=true
   ) {module=ScaleColor001}
   &RegionMask(Invert Region=true, Offset:Vector3f=<0 -60 0>, Region Falloff=0.2, Region Radius=160, UseOffset=true, UseRotation=false)
   &ScaleColor(
    Scale Alpha=&Multiply_Float(A=$Particles.RegionMask.Float, B=1)
    Scale RGB=&Multiply_VectorByFloat(Vector=<1 1 1>, Float=&FN_FrontEndFloatModifier(Gameplay=1, Front End=1.5))
    ScaleA=true
    ScaleRGB=true
   )
   &RegionMask(Invert Region=true, Offset:Vector3f=<0 -60 0>, Region Falloff=0, UseOffset=true, UseRotation=false) {module=RegionMask001}
   &ScaleMeshSize(
    Initial Mesh Scale:Vector3f=$Particles.Scale
    Scale Factor=&Multiply_VectorByFloat(Vector=<1 1 1>, Float=$Particles.RegionMask001.Float)
   )
  ]
  rendererClasses [&NiagaraMeshRendererProperties]
  renderers [{rendererClass=&NiagaraMeshRendererProperties, rendererIndex=0}]
 }
 first_piece_of_stairs {
  emitterUpdateScript [
   &EmitterState(
    Delay First Loop Only=true
    Inactive Response=::NewEnumerator1
    Life Cycle Mode=::NewEnumerator1
    Loop Delay=1.6
    Loop Duration=1
    UseLoopDelay=true
   )
   &SpawnRate(
    SpawnRate=&MakeCustomFloatFromBool(
     Bool=&SetBoolByFloatComparison(A=$User.ForwardDot, B=0, Comparison Type=::NewEnumerator6)
     True Float=0
     False Float=10
    )
    Use Spawn Probability=false
   )
  ]
  particleSpawnScript [
   &InitializeParticle(
    Color=<rgba 2.7653 2.6506 3 1>
    Lifetime Max=0.55
    Lifetime Min=0.5
    Lifetime Mode=::NewEnumerator1
    Mesh Scale Max=<20 20 1>
    Mesh Scale Min=<5 5 1>
    Mesh Scale Mode=::NewEnumerator2
    Sprite Rotation Mode=::NewEnumerator1
    Sprite Size Max=<10 1.5>
    Sprite Size Min=<40 1.5>
    Sprite Size Mode=::NewEnumerator2
    UsePositionOffset=false
    Write Material Random=true
   )
   &InitialMeshOrientation(Mesh Orientation Mode=::NewEnumerator2, Use Rotation Vector=true)
  ]
  particleUpdateScript [
   &ParticleState()
   &SkeletalMeshLocation(
    Bone Sampling Mode=::NewEnumerator4
    MaskSpawnGroup=false
    Orientation Sampling=::NewEnumerator1
    Sampled Position Offset=&Add_Vector(
     A=&Add_Vector(
      A=&Add_Vector(
       A=&RandomRangeVector(Minimum=<-60 0 0>, Maximum=<60 0 7>)
       B=&MakeVector(X=0, Y=&Lerp_Float(A=0, B=80, Alpha=$Particles.NormalizedAge), Z=0)
      )
      B=<0 -140 -20>
     )
     B=&MakeVector(
      X=0
      Y=&MakeCustomFloatFromBool(
       Bool=&SetBoolByFloatComparison(A=$User.ForwardDot, B=0, Comparison Type=::NewEnumerator6)
       True Float=&Lerp_Float(A=0, B=-30, Alpha=$Particles.NormalizedAge)
       False Float=0
      )
      Z=0
     )
    )
    Sampled Position Offset Coordinate Space:ENiagaraSkelMeshTransforms=::NewEnumerator2
    Skeletal Mesh=DataInterface() {propertyValues {
      DefaultMesh=null
      ExcludeBoneName=root
      OptionalComponentClass=null
      PreviewMesh=&Cube_Queen_Female_Glider
      SkinningMode=SkinOnTheFly
      SoftSourceActor=null
      SourceMode=Default
      UvSetIndex=0
      WholeMeshLOD=-1
      bExcludeBone=false
      bReadDeformedGeometry=true
      bRequireCurrentFrameData=true
      ComponentTags []
      FilteredBones [Root]
      FilteredSockets []
      MeshUserParameter {parameter {name=None, typeDefHandle {registeredTypeIndex=1}, varData []}}
      SamplingRegions []
     }}
   )
   &ScaleMeshSize(Scale Factor=&VectorFromFloat(Value=&FloatFromCurve(FloatCurve=DataInterface() {propertyValues {Curve {keys [
          {arriveTangent=3.4651, interpMode=RCIM_Cubic, leaveTangent=3.4651, tangentMode=RCTM_User, time=0, value=0}
          {arriveTangent=-0.0405, interpMode=RCIM_Cubic, leaveTangent=-0.0405, tangentMode=RCTM_User, time=1, value=1.5}
         ]}}})))
   &ScaleColor(
    Scale Alpha=&Multiply_Float(
     A=&FloatFromCurve(FloatCurve=DataInterface() {propertyValues {Curve {keys [
          {arriveTangent=-2.0467, interpMode=RCIM_Cubic, leaveTangent=-2.0467, tangentMode=RCTM_User, time=0, value=1}
          {arriveTangent=-0.281, interpMode=RCIM_Cubic, leaveTangent=-0.281, tangentMode=RCTM_User, time=1, value=0}
         ]}}})
     B=&FN_FrontEndFloatModifier(Gameplay=1, Front End=0.8)
    )
    ScaleA=true
    ScaleRGB=true
   )
   &DynamicMaterialParameters(
    Index 0 Param 1=&RandomRangeFloat(Minimum=0, Maximum=1)
    Index 0 Param 2=1
    Index 0 Param 3=0.5
    Index 1 Param 1=&RandomRangeFloat(Minimum=0.7, Maximum=1.2)
    Index 1 Param 2=&RandomRangeFloat(Minimum=0.7, Maximum=1.2)
    Module.Param0WriteEnabled=true
    Module.Param0WriteEnabled1=true
    Module.Param1WriteEnabled=true
    Module.Param1WriteEnabled1=true
    Module.Param2WriteEnabled=true
    Module.Param2WriteEnabled1=true
    Module.Param3WriteEnabled=true
    Module.Param3WriteEnabled1=true
    Write Parameter Index 1=true
   ) {module=DynamicMaterialParameters001}
  ]
  rendererClasses [&NiagaraMeshRendererProperties]
  renderers [{rendererClass=&NiagaraMeshRendererProperties, rendererIndex=0}]
 }
 first_piece_of_stairs_back {
  emitterUpdateScript [
   &EmitterState(
    Delay First Loop Only=true
    Inactive Response=::NewEnumerator1
    Life Cycle Mode=::NewEnumerator1
    Loop Delay=1.6
    Loop Duration=1
    UseLoopDelay=true
   )
   &SpawnRate(
    SpawnRate=&MakeCustomFloatFromBool(
     Bool=&SetBoolByFloatComparison(A=$User.ForwardDot, B=0, Comparison Type=::NewEnumerator6)
     True Float=10
     False Float=0
    )
    Use Spawn Probability=false
   )
  ]
  particleSpawnScript [
   &InitializeParticle(
    Color=<rgba 2.7653 2.6506 3 0.5>
    Lifetime Max=0.55
    Lifetime Min=0.5
    Lifetime Mode=::NewEnumerator1
    Mesh Scale Max=<20 20 1>
    Mesh Scale Min=<5 5 1>
    Mesh Scale Mode=::NewEnumerator2
    Sprite Rotation Mode=::NewEnumerator1
    Sprite Size Max=<10 1.5>
    Sprite Size Min=<40 1.5>
    Sprite Size Mode=::NewEnumerator2
    UsePositionOffset=false
    Write Material Random=true
   )
   &InitialMeshOrientation(Mesh Orientation Mode=::NewEnumerator2, Use Rotation Vector=true)
  ]
  particleUpdateScript [
   &ParticleState()
   &SkeletalMeshLocation(
    Bone Sampling Mode=::NewEnumerator4
    MaskSpawnGroup=false
    Orientation Sampling=::NewEnumerator1
    Sampled Position Offset=&Add_Vector(
     A=&Add_Vector(
      A=&Add_Vector(
       A=&RandomRangeVector(Minimum=<-60 0 0>, Maximum=<60 0 7>)
       B=&MakeVector(X=0, Y=&Lerp_Float(A=0, B=0, Alpha=$Particles.NormalizedAge), Z=0)
      )
      B=<0 -80 -20>
     )
     B=&MakeVector(
      X=0
      Y=&MakeCustomFloatFromBool(
       Bool=&SetBoolByFloatComparison(A=$User.ForwardDot, B=0, Comparison Type=::NewEnumerator6)
       True Float=&Lerp_Float(A=0, B=-80, Alpha=$Particles.NormalizedAge)
       False Float=0
      )
      Z=0
     )
    )
    Sampled Position Offset Coordinate Space:ENiagaraSkelMeshTransforms=::NewEnumerator2
    Skeletal Mesh=DataInterface() {propertyValues {
      DefaultMesh=null
      ExcludeBoneName=root
      OptionalComponentClass=null
      PreviewMesh=&Cube_Queen_Female_Glider
      SkinningMode=SkinOnTheFly
      SoftSourceActor=null
      SourceMode=Default
      UvSetIndex=0
      WholeMeshLOD=-1
      bExcludeBone=false
      bReadDeformedGeometry=true
      bRequireCurrentFrameData=true
      ComponentTags []
      FilteredBones [Root]
      FilteredSockets []
      MeshUserParameter {parameter {name=None, typeDefHandle {registeredTypeIndex=1}, varData []}}
      SamplingRegions []
     }}
   )
   &ScaleMeshSize(Scale Factor=&VectorFromFloat(Value=&FloatFromCurve(FloatCurve=DataInterface() {propertyValues {Curve {keys [
          {arriveTangent=-0.4621, interpMode=RCIM_Cubic, leaveTangent=-0.4621, tangentMode=RCTM_User, time=0, value=1.5}
          {arriveTangent=-3.3156, interpMode=RCIM_Cubic, leaveTangent=-3.3156, tangentMode=RCTM_User, time=1, value=0}
         ]}}})))
   &ScaleColor(
    Scale Alpha=&Multiply_Float(
     A=&FloatFromCurve(FloatCurve=DataInterface() {propertyValues {Curve {keys [
          {arriveTangent=-1.1343, interpMode=RCIM_Cubic, leaveTangent=-1.1343, tangentMode=RCTM_User, time=0, value=0}
          <at 0.3 0.65>
          {arriveTangent=-0.281, interpMode=RCIM_Cubic, leaveTangent=-0.281, tangentMode=RCTM_User, time=1, value=0}
         ]}}})
     B=&FN_FrontEndFloatModifier(Gameplay=1, Front End=0.8)
    )
    ScaleA=true
    ScaleRGB=true
   )
   &DynamicMaterialParameters(
    Index 0 Param 1=&RandomRangeFloat(Minimum=0, Maximum=1)
    Index 0 Param 2=1
    Index 0 Param 3=0.5
    Index 1 Param 1=&RandomRangeFloat(Minimum=0.7, Maximum=1.2)
    Index 1 Param 2=&RandomRangeFloat(Minimum=0.7, Maximum=1.2)
    Module.Param0WriteEnabled=true
    Module.Param0WriteEnabled1=true
    Module.Param1WriteEnabled=true
    Module.Param1WriteEnabled1=true
    Module.Param2WriteEnabled=true
    Module.Param2WriteEnabled1=true
    Module.Param3WriteEnabled=true
    Module.Param3WriteEnabled1=true
    Write Parameter Index 1=true
   ) {module=DynamicMaterialParameters001}
  ]
  rendererClasses [&NiagaraMeshRendererProperties]
  renderers [{rendererClass=&NiagaraMeshRendererProperties, rendererIndex=0}]
 }
}
summary {
 systemName=NS_Glider_Cube_Queen_Female
 emitters [
  {emitterName=Deploy_Stairs, rendererClasses [&NiagaraMeshRendererProperties]}
  {emitterName=Stair_fragments, rendererClasses [&NiagaraMeshRendererProperties]}
  {emitterName=Flame_Left, rendererClasses [&NiagaraMeshRendererProperties]}
  {emitterName=Flame_Right, rendererClasses [&NiagaraMeshRendererProperties]}
  {emitterName=Smoke, rendererClasses [&NiagaraSpriteRendererProperties]}
  {emitterName=The_first_half_of_the_stairs, rendererClasses [&NiagaraMeshRendererProperties]}
  {emitterName=The_second_half_of_the_stairs, rendererClasses [&NiagaraMeshRendererProperties]}
  {emitterName=first_piece_of_stairs, rendererClasses [&NiagaraMeshRendererProperties]}
  {emitterName=Range_cube, rendererClasses [&NiagaraMeshRendererProperties]}
  {emitterName=Character_cube, rendererClasses [&NiagaraMeshRendererProperties]}
  {emitterName=Deploy_Trail, rendererClasses [&NiagaraMeshRendererProperties]}
  {emitterName=Only_Frontend_Trail, rendererClasses [&NiagaraSpriteRendererProperties]}
  {emitterName=first_piece_of_stairs_back, rendererClasses [&NiagaraMeshRendererProperties]}
  {emitterName=Frontend_smoke, rendererClasses [&NiagaraSpriteRendererProperties]}
 ]
 userVariables [
  User.bIsFrontEnd {type=NiagaraBool}
  User.bIsFrontEndPreview {type=NiagaraBool}
  User.bisInVehicle {type=NiagaraBool}
  User.Flame_alpha {type=NiagaraFloat, defaultValue=0, description=""}
  User.Flame_Length {type=NiagaraFloat, defaultValue=0, description=""}
  User.ForwardDot {type=NiagaraFloat, defaultValue=0, description=""}
  User.Interpolate_WPO {type=NiagaraFloat, defaultValue=0, description=""}
  User.IsSkydiving {type=NiagaraBool}
  User.IsSwing {type=NiagaraBool}
  User.RotationalMovement {type=NiagaraFloat, defaultValue=0, description=""}
  User.SK_Mesh {
   type=NiagaraDataInterfaceSkeletalMesh
   defaultValue {
    struct=NiagaraExt_VariableValue_DataInterface
    value {dataInterface=&self:NiagaraDataInterfaceSkeletalMesh_0}
   }
  }
 ]
}
